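; Blueprint EQS context that supplies AI environment queries with every major water body actor in the world.
; ProvideActorsSet clears its output array, gathers all FortWaterBodyActor instances, and keeps only those whose water body type is River, Lake or Ocean.
; Each call does a full Get All Actors Of Class scan, and the querier inputs are ignored, so results are the same for any querier.
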(fn ProvideActorsSet
  (params (QuerierObject :type Object%20Reference) (QuerierActor :type Actor%20Object%20Reference))
  (Utilities|Array|Clear (Variables|Default|GetOutArray))
  (bind _outactors (Actor|GetAllActorsOfClass "/Script/FortniteGame.FortWaterBodyActor"))
  (for _array_element _outactors
    (if (Math|Boolean|ORBoolean (Utilities|Enum|Equal(Enum) (Water|GetWaterBodyType _array_element) "River") (Utilities|Enum|Equal(Enum) (Water|GetWaterBodyType _array_element) "Lake") (Utilities|Enum|Equal(Enum) (Water|GetWaterBodyType _array_element) "Ocean"))
      (Utilities|Array|Add (Variables|Default|GetOutArray) _array_element)))
  (return (Variables|Default|GetOutArray)))
