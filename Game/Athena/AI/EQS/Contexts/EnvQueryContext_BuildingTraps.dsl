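; EQS context for Athena AI that supplies the positions of every BuildingTrap actor in the world as a location set.
; ProvideLocationsSet clears its output array, gets all BuildingTrap actors, and adds each one's centroid with AddUnique, ignoring the querier.
; It uses centroids rather than actor locations because trap locations can sit on corners or other odd spots.
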(fn ProvideLocationsSet
  (params (QuerierObject :type Object%20Reference) (QuerierActor :type Actor%20Object%20Reference))
  (Utilities|Array|Clear (Variables|Default|GetOutCentroidLocations))
  (bind _outactors (Actor|GetAllActorsOfClass "/Script/FortniteGame.BuildingTrap"))
  (for _array_element _outactors
    (Utilities|Array|AddUnique (Variables|Default|GetOutCentroidLocations) (Building|GetCentroid _array_element) :!pinned true) ; We use the Centroid instead of Location because the Location can be on corners and weird places.\n\nCentroid will feel more correct.
  )
  (return (Variables|Default|GetOutCentroidLocations)))
