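; An EQS context that supplies foliage plant and tall grass actors to AI environment queries.
; It overrides ProvideActorsSet to clear an output array, then append every world actor of three classes: Asteria medium foliage plants, jungle foliage clusters and Artemis stealth grass.
; Each call uses GetAllActorsOfClass with no filtering by querier, distance or state.
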(fn ProvideActorsSet
  (params (QuerierObject :type Object%20Reference) (QuerierActor :type Actor%20Object%20Reference))
  (Utilities|Array|Clear (Variables|Default|GetOutArray))
  (bind _outactors (Actor|GetAllActorsOfClass "/Game/Environments/Asteria/Foliage/Blueprints/BP_Medium_Foliage_Plant_Parent.BP_Medium_Foliage_Plant_Parent_C"))
  (Utilities|Array|AppendArray (Variables|Default|GetOutArray) _outactors)
  (bind _outactors_2 (Actor|GetAllActorsOfClass "/Game/Environments/Asteria/Foliage/JunglePlants/FoliageCluster/BP_FoliageCluster.BP_FoliageCluster_C"))
  (Utilities|Array|AppendArray (Variables|Default|GetOutArray) _outactors_2)
  (bind _outactors_3 (Actor|GetAllActorsOfClass "/Game/Athena/Artemis/Environments/BuildingActors/Foliage/B_Artemis_StealthGrass.B_Artemis_StealthGrass_C"))
  (Utilities|Array|AppendArray (Variables|Default|GetOutArray) _outactors_3)
  (return (Variables|Default|GetOutArray)))
