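; EQS context that supplies the bottom point of the querying character's collision capsule as a single location.
; ProvideSingleLocation casts the querier actor to Character and checks that its CapsuleComponent is valid.
; It returns the capsule's world location minus its half height on Z, placing the point at the character's feet.
; No location is explicitly returned when the capsule is invalid.
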(fn ProvideSingleLocation
  (params (QuerierObject :type Object%20Reference) (QuerierActor :type Actor%20Object%20Reference))
  (bind _ascharacter (Utilities|Casting|CastToCharacter QuerierActor))
  (Utilities|IsValid (Variables|/Script/Engine.Character|GetCapsuleComponent _ascharacter)
    (:Is%20Valid
      (return (Utilities|Operators|Subtract (Transformation|GetWorldLocation (Variables|/Script/Engine.Character|GetCapsuleComponent _ascharacter)) (Math|Vector|MakeVector 0.0 0.0 (Variables|/Script/Engine.CapsuleComponent|GetCapsuleHalfHeight (Variables|/Script/Engine.Character|GetCapsuleComponent _ascharacter))))))))
