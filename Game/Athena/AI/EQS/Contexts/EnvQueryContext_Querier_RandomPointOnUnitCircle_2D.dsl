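; EQS context for Athena AI that provides a random location one unit from the querier on the horizontal plane.
; It builds a vector from the X and Y of two separate random unit vector calls, normalizes it and adds it to the querier actor's location.
; Its purpose is to randomize EQS generation directions when used as the target of an arc direction or two-point direction from the querier.
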(fn ProvideSingleLocation
  (params (QuerierObject :type Object%20Reference) (QuerierActor :type Actor%20Object%20Reference))
  ;[1 1 1] Make a point on a 1 unit circle around the Querier, randomly.
  ;
  ; This allows us to use ArcDirection/TwoPoints from Querier to this Context to randomize EQS generation directions.
  (return (+ (Transformation|GetActorLocation QuerierActor) (Math|Vector|Normalize (Math|Vector|MakeVector (.x (Math|Random|RandomUnitVector)) (.y (Math|Random|RandomUnitVector)))))))
