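; An EQS query context that supplies the querier's location snapped to the Fortnite building grid, at a floor corner.
; It overrides ProvideSingleLocation, passing the querier actor's location through SnaptoFortBuildingGrid with an XY offset.
; Snapping lets EQS generators place their points more consistently.
; An invalid querier actor yields the zero vector instead.
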(fn ProvideSingleLocation
  (params (QuerierObject :type Object%20Reference) (QuerierActor :type Actor%20Object%20Reference))
  ;[1 1 1] Snap the Querier's location to the nearest building grid square so Generators place points more consistently.
  (Utilities|IsValid QuerierActor
    (:Is%20Valid
      (return (Fortnite|SnaptoFortBuildingGrid (Transformation|GetActorLocation QuerierActor) "0, 0, 0" "256.000000,256.000000,0.000000")))
    (:Is%20Not%20Valid
      (return "0, 0, 0"))))
